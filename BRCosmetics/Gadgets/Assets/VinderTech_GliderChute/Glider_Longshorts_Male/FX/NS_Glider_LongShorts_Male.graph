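# Niagara effect for the LongShorts Male glider, adding cyan engine flames, idle flames and trails driven by the glider's skeletal mesh.
# EngineFlames keeps burst-spawned mesh particles and a light attached to the mesh, stretching layered flame cores by a system ForwardFlame value.
# IdleFlames emits shrinking, fading sprites from skeleton samples, while FN_AnimTrail draws two bone-based ribbons.
# Full deployment gates all three emitters, and trails also require a ForwardDot threshold.
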NiagaraSystem /BRCosmetics/Gadgets/Assets/VinderTech_GliderChute/Glider_Longshorts_Male/FX/NS_Glider_LongShorts_Male
dependencies {
 usedDataInterfaces [&NiagaraDataInterfaceCurve, &NiagaraDataInterfaceColorCurve]
 usedDynamicInputs []
 usedModules [
  &SystemState
  &FN_Integrator
  &self:SystemSpawnScript.SystemScriptSource.SystemScriptGraph.NiagaraNodeAssignment_0.SetVariables_#1
 ]
}
emitters {
 EngineFlames {
  emitterSpawnScript [&FN_PreserveSkydiveUserVariables() {module=PreserveUserVariables}]
  emitterUpdateScript [
   &EmitterState(Life Cycle Mode=::NewEnumerator1, Loop Behavior=::NewEnumerator1, Loop Duration=1, UseLoopDelay=false)
   &SpawnBurst_Instantaneous(Spawn Count=1, Use Spawn Probability=true)
  ]
  particleSpawnScript [
   /Niagara/Modules/Spawn/Initialization/InitializeParticle(
    Lifetime=0.25
    Mesh Scale=<0.25 0.125 0.65>
    Module.Write Color=true
    Module.Write Lifetime=true
    Module.Write Mass=false
    Module.Write Material Random=true
    Module.Write Position=false
    Module.Write Scale=true
    Module.Write SpriteRotation=false
    Module.Write SpriteSize=false
   )
   /Niagara/Modules/Spawn/Orientation/V2/InitialMeshOrientation(Rotation=<0.25 0 0>, Use Orientation Vector=true, Use Rotation Vector=true) {enabled=false}
   /Niagara/Modules/Spawn/Orientation/V3/InitialMeshOrientation(Rotation=<0 -0.25 0>, Use Orientation Vector=true, Use Rotation Vector=true) {module=InitialMeshOrientation001}
   /SetVariables(Particles.FlameCoreScale:Vector3f=<2 0.2 0.2>) {isSetParameters=true}
   /SetVariables(Particles.FlameCoreOrientation:Quat4f=<0 0 -1 1>) {isSetParameters=true}
   /SetVariables(Particles.FlameCoreSecondaryScale:Vector3f=<2.25 0.5 0.5>) {isSetParameters=true}
   /SetVariables(Particles.FlameCoreSecondaryOrientation:Quat4f=<0 0 1 1>) {isSetParameters=true}
  ]
  particleUpdateScript [
   &ParticleState(Kill Particles When Lifetime Has Elapsed=false)
   &SampleSkeletalMeshSkeleton(
    Mesh:NiagaraDataInterfaceSkeletalMesh=$User.SkeletalMesh
    Spawn Mode:ENiagaraSkeletalSpawning=::NewEnumerator1
    Specific Socket Index:NiagaraInt32=&MakeIntFromFloat(Float=0)
   ) {enabled=false}
   /Niagara/Modules/Spawn/Location/SkeletalMeshSkeletonLocation(Module.Inherit Velocity=false, Module.Use Sampled Bone as Ribbon ID=true, Set Mesh Orientation to Bone Orientation=false) {enabled=false}
   &SkeletalMeshLocation(
    Apply Sampled Orientation to Mesh Particles=true
    Inherit Velocity=false
    MaskSpawnGroup=false
    Mesh Sampling Type=::NewEnumerator5
    Skeletal Mesh=$User.SkeletalMesh
    Socket Sampling Mode=::NewEnumerator5
   )
   &SolveForcesAndVelocity(Clamp Velocity=false, Limit Acceleration=false)
   &ScaleColor(
    Scale Alpha=&FloatFromCurve(
     FloatCurve=DataInterface() {propertyValues {Curve {keys [<at 0 0 RCIM_Cubic>,<at 0.5 1 RCIM_Cubic>]}}}
     Scale Curve=&MakeCustomFloatFromBool(Bool=$User.bIsFullyDeployed, True Float=1, False Float=0)
    )
    ScaleA=true
    ScaleRGB=false
   )
   /SetVariables(Particles.FlameCoreColor:LinearColor=&MakeLinearColorFromVectorAndFloat(
     "Vector (RGB)"=&MakeVectorFromLinearColorRGB(Color=<rgba 0 2.4444 4 1>)
     "Float (Alpha)"=&MakeCustomFloatFromBool(Bool=$User.bIsFullyDeployed, True Float=1, False Float=0)
    )) {isSetParameters=true}
   /SetVariables(Particles.FlameCoreSecondaryColor:LinearColor=&MakeLinearColorFromVectorAndFloat(
     "Vector (RGB)"=&MakeVectorFromLinearColorRGB(Color=<rgba 0 2.4444 4 0.25>)
     "Float (Alpha)"=&MakeCustomFloatFromBool(Bool=$User.bIsFullyDeployed, True Float=0.25, False Float=0)
    )) {isSetParameters=true}
   &ScaleMeshSize(Scale Factor=&MakeVector(
     X=1
     Y=1
     Z=&FloatFromCurve(FloatCurve=DataInterface() {propertyValues {Curve {keys [<at 0 0>,<at 0.85 1>]}}})
    ))
   &ScaleMeshSize(Scale Factor=&MakeVector(X=1, Y=1, Z=$System.ForwardFlame)) {module=ScaleMeshSize001}
   /SetVariables(Particles.FlameCoreScale:Vector3f=&Multiply_Vector(
     A=$Particles.Initial.FlameCoreScale
     B=&MakeVector(
      X=&FloatFromCurve(FloatCurve=DataInterface() {propertyValues {Curve {keys [<at 0 0>,<at 1 1>]}}})
      Y=1
      Z=1
     )
    )) {isSetParameters=true}
   /SetVariables(Particles.FlameCoreScale:Vector3f=&Multiply_Vector(A=$Particles.FlameCoreScale, B=&MakeVector(X=$System.ForwardFlame, Y=1, Z=1))) {isSetParameters=true}
   /SetVariables(Particles.FlameCoreSecondaryScale:Vector3f=&Multiply_Vector(
     A=$Particles.Initial.FlameCoreSecondaryScale
     B=&MakeVector(
      X=&FloatFromCurve(FloatCurve=DataInterface() {propertyValues {Curve {keys [<at 0 0>,<at 1 1>]}}})
      Y=1
      Z=1
     )
    )) {isSetParameters=true}
   /SetVariables(Particles.FlameCoreSecondaryScale:Vector3f=&Multiply_Vector(A=$Particles.FlameCoreSecondaryScale, B=&MakeVector(X=$System.ForwardFlame, Y=1, Z=1))) {isSetParameters=true}
   &FN_Light_v2(Light Exponent=1, Light Radius=50, LightEnabled:NiagaraBool=$User.bIsFullyDeployed, Scale Color=1)
   /SetVariables(Particles.Light Color:LinearColor=<rgba 0 2.7467 4 1>) {isSetParameters=true}
  ]
  rendererClasses [&NiagaraMeshRendererProperties, &NiagaraLightRendererProperties]
  renderers [
   {rendererClass=&NiagaraMeshRendererProperties, rendererIndex=0}
   {rendererClass=&NiagaraMeshRendererProperties, rendererIndex=1}
   {rendererClass=&NiagaraMeshRendererProperties, rendererIndex=2}
   {rendererClass=&NiagaraLightRendererProperties, rendererIndex=3}
  ]
 }
 FN_AnimTrail {
  emitterSpawnScript [&FN_RibbonIDOffsetUtility()]
  emitterUpdateScript [
   &EmitterState()
   &FN_SpawnAnimTrail(
    Enable:NiagaraBool=&Boolean_AndOperation(
     Bool1=$User.bIsFullyDeployed
     Bool2=&SetBoolByFloatComparison(A=$User.ForwardDot, B=0.2, Comparison Type=::NewEnumerator2)
    )
    Number of Anim Trails=2
   )
  ]
  particleSpawnScript [
   &FN_InitializeAnimTrailBoneIndecies(First Bone or Socket Index=2)
   &FN_InitializeAnimTrailBoneIndecies(AnimTrail Selection=1, First Bone or Socket Index=3, Second Bone or Socket Index=4) {module=FN_InitializeAnimTrailBoneIndecies001}
   &FN_InitializeAnimTrail(AnimTrailWidth=1.75, Color=<rgba 1 1 1 0.0001>, Lifetime=0.2)
   &AddVelocity(Velocity=<0 400 0>) {enabled=false}
  ]
  particleUpdateScript [
   &SolveForcesAndVelocity(Clamp Velocity=false, Limit Acceleration=false)
   /Game/Effects/Niagara/Modules/Materials/Material_v2()
   &FN_UpdateAnimTrail(Locked Particle RibbonWidth=1, SkeletalMesh=$User.SkeletalMesh) {module=FN_UpdateAnimTrail001}
   /SetVariables(Particles.RibbonU0Override=$Particles.NormalizedAge) {isSetParameters=true}
   &ScaleColor(Scale Alpha=$User.LocalAlpha, ScaleA=true, ScaleRGB=false) {enabled=false}
   &ScaleRibbonWidth(Ribbon Width Scale=&FloatFromCurve(FloatCurve=DataInterface() {propertyValues {Curve {keys [<at 0 1>,<at 0.5 0.5>]}}}))
  ]
  rendererClasses [&NiagaraRibbonRendererProperties]
  renderers [{rendererClass=&NiagaraRibbonRendererProperties}]
 }
 IdleFlames {
  emitterSpawnScript [&FN_PreserveSkydiveUserVariables() {module=PreserveUserVariables}]
  emitterUpdateScript [
   &EmitterState()
   &SpawnRate(
    Spawn Probability=&MakeCustomFloatFromBool(Bool=$User.bIsFullyDeployed, True Float=1, False Float=0)
    SpawnRate=16
    Use Spawn Probability=true
   )
  ]
  particleSpawnScript [&InitializeParticle(
    Color=<rgba 0 0.687 1 4>
    Lifetime=0.8
    Module.Write Color=true
    Module.Write Lifetime=true
    Module.Write Mass=false
    Module.Write Material Random=true
    Module.Write Position=false
    Module.Write Scale=false
    Module.Write SpriteRotation=false
    Module.Write SpriteSize=true
    Sprite Size=<14 10>
   )]
  particleUpdateScript [
   &ParticleState()
   &SampleSkeletalMeshSkeleton(
    Mesh:NiagaraDataInterfaceSkeletalMesh=$User.SkeletalMesh
    Spawn Mode:ENiagaraSkeletalSpawning=::NewEnumerator1
    Specific Socket Index:NiagaraInt32=&MakeIntFromFloat(Float=0)
   )
   &SolveForcesAndVelocity(Clamp Velocity=false, Limit Acceleration=false)
   &ScaleColor(
    Scale Alpha=&FloatFromCurve(FloatCurve=DataInterface() {propertyValues {Curve {keys [<at 0 0 RCIM_Cubic>, <at 0.5 1 RCIM_Cubic>, {interpMode=RCIM_Cubic, leaveTangent=0.2, time=1, value=0}]}}})
    Scale RGB=&MakeVectorFromLinearColorRGB(Color=&ColorFromCurve(ColorCurve=DataInterface() {propertyValues {
        AlphaCurve {keys [<at -0.0023 1 RCIM_Cubic>,<at 0 1 RCIM_Cubic>]}
        BlueCurve {keys [<at -0.002 4 RCIM_Cubic>,<at 0.998 1 RCIM_Cubic>]}
        GreenCurve {keys [<at -0.002 4 RCIM_Cubic>,<at 0.998 0.7389 RCIM_Cubic>]}
        RedCurve {keys [<at -0.002 4 RCIM_Cubic>,<at 0.998 0.6237 RCIM_Cubic>]}
       }}))
    ScaleA=true
    ScaleRGB=true
   )
   &ScaleSpriteSize(Scale Factor=&Vector2DFromFloat(Value=&FloatFromCurve(FloatCurve=DataInterface() {propertyValues {Curve {keys [<at 0 1>,<at 1 0.5>]}}})))
   /SetVariables(Particles.OriginPosition:Vector3f=$Particles.SampleSkeletalMeshSkeleton.SampledPosition) {isSetParameters=true}
   /SetVariables() {isSetParameters=true}
  ]
  rendererClasses [&NiagaraSpriteRendererProperties]
  renderers [{rendererClass=&NiagaraSpriteRendererProperties}]
 }
}
summary {
 systemName=NS_Glider_LongShorts_Male
 emitters [
  {emitterName=EngineFlames, rendererClasses [&NiagaraMeshRendererProperties, &NiagaraLightRendererProperties]}
  {emitterName=IdleFlames, rendererClasses [&NiagaraSpriteRendererProperties]}
  {emitterName=FN_AnimTrail, rendererClasses [&NiagaraRibbonRendererProperties]}
 ]
 userVariables [
  User.bIsFrontEndPreview {type=NiagaraBool}
  User.bIsFullyDeployed {type=NiagaraBool}
  User.ForwardDot {type=NiagaraFloat, defaultValue=0, description=""}
  User.ForwardFlame {type=NiagaraFloat, defaultValue=1, description=""}
  User.LocalAlpha {type=NiagaraFloat, defaultValue=0, description=""}
  User.PlayerSpeed {type=NiagaraFloat, defaultValue=0, description=""}
  User.RibbonIDOffset {type=NiagaraInt32}
  User.SkeletalMesh {
   type=NiagaraDataInterfaceSkeletalMesh
   defaultValue {
    struct=NiagaraExt_VariableValue_DataInterface
    value {dataInterface=&self:NiagaraDataInterfaceSkeletalMesh_0}
   }
  }
 ]
}
